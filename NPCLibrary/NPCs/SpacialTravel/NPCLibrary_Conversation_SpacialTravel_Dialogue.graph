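# Conversation database for the SpacialTravel NPC's "Ask about..." dialogue, voiced as a Light Rider character from the Supernovas.
# A single entry tag opens a greeting Speech node that branches into Powers, Supernova Academy, Teammates and a Back choice.
# Each topic plays a linear chain of Speech lines with lore on that subject.
# All three chains converge on one shared ReturnToStart node, shown as BACK, while the greeting's Back choice uses a separate node.
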ConversationDatabase /NPCLibrary/NPCs/SpacialTravel/NPCLibrary_Conversation_SpacialTravel_Dialogue
entry_points [{tag=Conversation.Entry.SpacialTravel.Dialogue, destinations [#1]}]
nodes [
 ConversationEntryPointNode_0 {
  id=#1
  class=ConversationEntryPointNode
  outputs [#2]
  properties {entryTag {tagName=Conversation.Entry.SpacialTravel.Dialogue}}
 }
 Speech_C_0 {
  id=#2
  class=Speech_C
  outputs [#3, #4, #5, #6]
  properties {nodeName="", softTaskIcon=&T_UI_NPC_question, generalConfig {defaultMessage=I guess I have a minute. What's up?}}
  sub_nodes [ConversationChoiceNode_5 {class=ConversationChoiceNode, properties {defaultChoiceDisplayText=Ask about...}}]
 }
 Speech_C_5 {
  id=#6
  class=Speech_C
  outputs [#7]
  properties {
   nodeName=""
   softTaskIcon=&T_UI_NPC_question
   generalConfig {defaultMessage="If I'm honest, I underestimated them. All of them."}
  }
  sub_nodes [ConversationChoiceNode_15 {class=ConversationChoiceNode, properties {defaultChoiceDisplayText=Teammates}}]
 }
 Speech_C_6 {
  id=#7
  class=Speech_C
  outputs [#8]
  properties {
   nodeName=""
   softTaskIcon=&T_UI_NPC_question
   generalConfig {defaultMessage="Everyone contributes: Killswitch as tactician, Synthwave as support, Haylee as tank, Ziggy as support..."}
  }
 }
 Speech_C_8 {
  id=#8
  class=Speech_C
  outputs [#9]
  properties {
   nodeName=""
   softTaskIcon=&T_UI_NPC_question
   generalConfig {defaultMessage=And Morgan... She's really stepped up as a leader. Really.}
  }
 }
 Speech_C_10 {
  id=#9
  class=Speech_C
  outputs [#10]
  properties {
   nodeName=""
   softTaskIcon=&T_UI_NPC_question
   generalConfig {defaultMessage=I never wanted to rely on anyone in case they let me down. But the Supernovas have my back.}
  }
 }
 ReturnToStart_C_23 {
  id=#10
  class=ReturnToStart_C
  properties {softTaskIcon=&T_UI_NPC_Back}
  sub_nodes [ConversationChoiceNode_32 {class=ConversationChoiceNode, properties {defaultChoiceDisplayText=BACK}}]
 }
 Speech_C_1 {
  id=#5
  class=Speech_C
  outputs [#11]
  properties {
   nodeName=""
   softTaskIcon=&T_UI_NPC_question
   generalConfig {defaultMessage=I didn't really know how much I'd miss the Academy until it was taken away from us.}
  }
  sub_nodes [ConversationChoiceNode_1 {class=ConversationChoiceNode, properties {defaultChoiceDisplayText=Supernova Academy}}]
 }
 Speech_C_2 {
  id=#11
  class=Speech_C
  outputs [#12]
  properties {
   nodeName=""
   softTaskIcon=&T_UI_NPC_question
   generalConfig {defaultMessage="The classes, the dorms, the routines. Even Hope's touchy-feely lessons."}
  }
 }
 Speech_C_4 {
  id=#12
  class=Speech_C
  outputs [#13]
  properties {
   nodeName=Speech
   softTaskIcon=&T_UI_NPC_question
   generalConfig {defaultMessage="I still know every jump, every corner, every angle of those buildings."}
  }
 }
 Speech_C_9 {
  id=#13
  class=Speech_C
  outputs [#10]
  properties {
   nodeName=Speech
   softTaskIcon=&T_UI_NPC_question
   generalConfig {defaultMessage="Things just aren't the same without... well, our home."}
  }
 }
 Back_C_0 {
  id=#4
  class=Back_C
  sub_nodes [ConversationChoiceNode_7 {class=ConversationChoiceNode, properties {defaultChoiceDisplayText=BACK}}]
 }
 Speech_C_7 {
  id=#3
  class=Speech_C
  outputs [#14]
  properties {
   nodeName=""
   softTaskIcon=&T_UI_NPC_question
   generalConfig {defaultMessage=It's all in the name: Light. Rider. I ride light. Easy.}
  }
  sub_nodes [ConversationChoiceNode_0 {class=ConversationChoiceNode, properties {defaultChoiceDisplayText=Powers}}]
 }
 Speech_C_15 {
  id=#14
  class=Speech_C
  outputs [#15]
  properties {
   nodeName=""
   softTaskIcon=&T_UI_NPC_question
   generalConfig {defaultMessage="Specifically, I surf on waves of pure rezz energy.`"}
  }
 }
 Speech_C_14 {
  id=#15
  class=Speech_C
  outputs [#10]
  properties {
   nodeName=""
   softTaskIcon=&T_UI_NPC_question
   generalConfig {defaultMessage="What's 'rezz energy'? I don't know, do I look like a scientist?"}
  }
 }
]
